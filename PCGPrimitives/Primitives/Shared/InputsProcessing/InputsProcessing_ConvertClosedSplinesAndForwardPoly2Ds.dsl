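; PCG graph for shared primitive input processing that converts closed splines to Polygon2Ds and forwards existing Polygon2Ds on the same output.
; ClosedSplines pass through Create Polygon 2D with ForceClosed, then have their Type tags stripped and replaced with Type:Polygon2D.
; Incoming Polygon2Ds go straight to Out with their tags untouched.
; The Points, Splines, Surfaces, Texture2Ds and Volumes inputs are declared but unused.
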(graph InputsProcessing_ConvertClosedSplinesAndForwardPoly2Ds
  (inputs (Points :type Point) (Splines :type Spline) (ClosedSplines :type Spline) (Polygon2Ds :type Polygon2D) (Surfaces :type Surface) (Texture2Ds :type BaseTexture) (Volumes :type Volume))
  ;[0 0 0 0.5] Convert ClosedSplines To Polygon2Ds and Forward Existing Polygon2Ds
  (bind CreatePolygon2D_0 (Create%20Polygon%202D :InputType ForceClosed))
  (bind DeleteTags_0 (Delete%20Tags :SelectedTags "Type:Point, Type:Spline, Type:ClosedSpline, Type:Polygon2D, Type:Surface, Type:Texture2D, Type:Volume"))
  (bind AddTags_0 (Add%20Tags :TagsToAdd "Type:Polygon2D"))
  (connect DeleteTags_0 Out AddTags_0 In)
  (connect DefaultInputNode ClosedSplines CreatePolygon2D_0 In)
  (connect CreatePolygon2D_0 Out DeleteTags_0 In)
  (output Out AddTags_0)
  (output Out Polygon2Ds))
