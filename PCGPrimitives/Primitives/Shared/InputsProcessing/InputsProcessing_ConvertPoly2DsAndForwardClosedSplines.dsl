(graph InputsProcessing_ConvertPoly2DsAndForwardClosedSplines
  (params (KeepPolygon2DsHoles :type Bool))
  (inputs (Points :type Point) (Splines :type Spline) (ClosedSplines :type Spline) (Polygon2Ds :type Polygon2D) (Surfaces :type Surface) (Texture2Ds :type BaseTexture) (Volumes :type Volume))
  ;[0 0 0 0.5] Convert Polygon2Ds (excluding holes) To ClosedSplines and Forward Existing ClosedSplines
  (bind ToPoint_0 (To%20Point))
  (bind Branch_1 (Branch :bOutputToB-param KeepPolygon2DsHoles))
  (bind AttributeFilter_1 (Filter%20Attribute%20Elements :AttributeTypes (struct :Type Integer32 :Int32Value -1) :bUseConstantThreshold true :Operator Equal :TargetAttribute HoleIndex :ThresholdAttribute @Last))
  (bind CreateSpline_0 (Create%20Spline :bClosedLoop true :bLinear true :InterpTypeAttribute None))
  (bind DeleteTags_1 (Delete%20Tags :Operator Substring :SelectedTags "Type:Point, Type:Spline, Type:ClosedSpline, Type:Polygon2D, Type:Surface, Type:Texture2D, Type:Volume"))
  (bind AddTags_1 (Add%20Tags :TagsToAdd "Type:ClosedSpline"))
  (connect DeleteTags_1 Out AddTags_1 In)
  (connect Branch_1 Output%20A AttributeFilter_1 In)
  (connect ToPoint_0 Out Branch_1 In)
  (connect AttributeFilter_1 InsideFilter CreateSpline_0 In)
  (connect Branch_1 Output%20B CreateSpline_0 In)
  (connect CreateSpline_0 Out DeleteTags_1 In)
  (connect DefaultInputNode Polygon2Ds ToPoint_0 In)
  (output Out AddTags_1)
  (output Out ClosedSplines))
