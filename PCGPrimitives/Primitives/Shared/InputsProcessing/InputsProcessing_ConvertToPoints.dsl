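; PCG subgraph in the shared input-processing primitives that converts Splines, ClosedSplines and Polygon2Ds to points.
; Splines and ClosedSplines become their control points, while Polygon2Ds go through To Point.
; Converted data has its Type tags removed and is retagged Type:Point, then joins the untouched Points input on Out.
; Surfaces, Texture2Ds and Volumes are declared as inputs but left unconnected, so they are not passed on.
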(graph InputsProcessing_ConvertToPoints
  (inputs (Points :type Point) (Splines :type Spline) (ClosedSplines :type Spline) (Polygon2Ds :type Polygon2D) (Surfaces :type Surface) (Texture2Ds :type BaseTexture) (Volumes :type Volume))
  ;[0 0 0 0.5] Forward Points, Convert Splines, ClosedSplines and Polygon2Ds to Points
  (bind GetSplineControlPoints_0 (Get%20Spline%20Control%20Points))
  (bind ToPoint_0 (To%20Point))
  (bind DeleteTags_0 (Delete%20Tags :SelectedTags "Type:Point, Type:Spline, Type:ClosedSpline, Type:Polygon2D, Type:Surface, Type:Texture2D, Type:Volume"))
  (bind AddTags_0 (Add%20Tags :TagsToAdd "Type:Point"))
  (connect DeleteTags_0 Out AddTags_0 In)
  (connect GetSplineControlPoints_0 Out DeleteTags_0 In)
  (connect ToPoint_0 Out DeleteTags_0 In)
  (connect DefaultInputNode ClosedSplines GetSplineControlPoints_0 In)
  (connect DefaultInputNode Splines GetSplineControlPoints_0 In)
  (connect DefaultInputNode Polygon2Ds ToPoint_0 In)
  (output Out AddTags_0)
  (output Out Points))
