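; PCG subgraph that forces mixed shape inputs into splines, passing Splines and ClosedSplines through and ignoring points, surfaces, textures and volumes.
; Polygon2Ds become points, then closed linear splines that pass through Spline Direction.
; Unless KeepPolygon2DsHoles is set, a HoleIndex == -1 filter drops hole points first.
; Converted splines have their "Type:" tags replaced with "Type:ClosedSpline" before joining the single output.
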(graph InputsProcessing_ForceSpline
  (params (KeepPolygon2DsHoles :type Bool))
  (inputs (Points :type Point) (Splines :type Spline) (ClosedSplines :type Spline) (Polygon2Ds :type Polygon2D) (Surfaces :type Surface) (Texture2Ds :type BaseTexture) (Volumes :type Volume))
  ;[0 0 0 0.5] Forward Splines and Closed Splines, convert Polygon2Ds to Closed Splines
  (bind ToPoint_0 (To%20Point))
  (bind Branch_0 (Branch :bOutputToB-param KeepPolygon2DsHoles))
  (bind AttributeFilter_0 (Filter%20Attribute%20Elements :AttributeTypes (struct :Type Integer32 :Int32Value -1) :bUseConstantThreshold true :Operator Equal :TargetAttribute HoleIndex :ThresholdAttribute @Last))
  (bind CreateSpline_0 (Create%20Spline :bClosedLoop true :bLinear true :InterpTypeAttribute None))
  (bind SplineDirection_0 (Spline%20Direction))
  (bind DeleteTags_1 (Delete%20Tags :Operator Substring :SelectedTags "Type:Point, Type:Spline, Type:ClosedSpline, Type:Polygon2D, Type:Surface, Type:Texture2D, Type:Volume"))
  (bind AddTags_1 (Add%20Tags :TagsToAdd "Type:ClosedSpline"))
  (connect DeleteTags_1 Out AddTags_1 In)
  (connect Branch_0 Output%20A AttributeFilter_0 In)
  (connect ToPoint_0 Out Branch_0 In)
  (connect AttributeFilter_0 InsideFilter CreateSpline_0 In)
  (connect Branch_0 Output%20B CreateSpline_0 In)
  (connect SplineDirection_0 Out DeleteTags_1 In)
  (connect CreateSpline_0 Out SplineDirection_0 In)
  (connect DefaultInputNode Polygon2Ds ToPoint_0 In)
  (output Out AddTags_1)
  (output Out ClosedSplines)
  (output Out Splines))
