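; PCG graph that converts mixed shape inputs into surface data as a shared input-processing step.
; Splines become surfaces via 2D polygons, Polygon2Ds convert directly, closed splines use Create Surface From Spline, and Texture2Ds skip conversion.
; These results have their Type tags replaced with Type:Surface, while incoming Surfaces pass straight to the output.
; Points and Volumes inputs are left unconnected.
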(graph InputsProcessing_ForceSurface
  (inputs (Points :type Point) (Splines :type Spline) (ClosedSplines :type Spline) (Polygon2Ds :type Polygon2D) (Surfaces :type Surface) (Texture2Ds :type BaseTexture) (Volumes :type Volume))
  ;[0 0 0 0.5] Inputs Processing and Transformation
  (bind CreatePolygon2D_0 (Create%20Polygon%202D))
  (bind CreateSurfaceFromPolygon2D_0 (Create%20Surface%20From%20Polygon2D))
  (bind CreateSurfaceFromSpline_0 (Create%20Surface%20From%20Spline))
  (bind DeleteTags_0 (Delete%20Tags :Operator Substring :SelectedTags "Type:Point, Type:Spline, Type:ClosedSpline, Type:Polygon2D, Type:Surface, Type:Texture2D, Type:Volume"))
  (bind AddTags_0 (Add%20Tags :TagsToAdd "Type:Surface"))
  (connect DeleteTags_0 Out AddTags_0 In)
  (connect DefaultInputNode Splines CreatePolygon2D_0 In)
  (connect CreatePolygon2D_0 Out CreateSurfaceFromPolygon2D_0 In)
  (connect DefaultInputNode Polygon2Ds CreateSurfaceFromPolygon2D_0 In)
  (connect DefaultInputNode ClosedSplines CreateSurfaceFromSpline_0 In)
  (connect CreateSurfaceFromPolygon2D_0 Out DeleteTags_0 In)
  (connect CreateSurfaceFromSpline_0 Out DeleteTags_0 In)
  (connect DefaultInputNode Texture2Ds DeleteTags_0 In)
  (output Out AddTags_0)
  (output Out Surfaces))
