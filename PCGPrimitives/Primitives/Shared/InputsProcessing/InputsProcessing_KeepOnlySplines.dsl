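; PCG input-processing graph that keeps only spline data from a set of typed inputs.
; It declares point, spline, closed spline, polygon, surface, texture and volume inputs, but only Splines is wired through.
; That data has any existing Type tags stripped, then receives a Type:Spline tag before reaching the single output.
; Closed splines arriving on their separate pin are dropped along with the other types.
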(graph InputsProcessing_KeepOnlySplines
  (inputs (Points :type Point) (Splines :type Spline) (ClosedSplines :type Spline) (Polygon2Ds :type Polygon2D) (Surfaces :type Surface) (Texture2Ds :type BaseTexture) (Volumes :type Volume))
  ;[0 0 0 0.5] Inputs Processing and Transformation
  (bind DeleteTags_0 (Delete%20Tags :SelectedTags "Type:Point, Type:Spline, Type:ClosedSpline, Type:Polygon2D, Type:Surface, Type:Texture2D, Type:Volume"))
  (bind AddTags_0 (Add%20Tags :TagsToAdd "Type:Spline"))
  (connect DeleteTags_0 Out AddTags_0 In)
  (connect DefaultInputNode Splines DeleteTags_0 In)
  (output Out AddTags_0))
