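; PCG input-processing subgraph that normalizes points, splines or existing polygons into Polygon2D data on a single output.
; Point and spline data from Input, plus Areas and Paths, feed Create Polygon 2D.
; A polygon count over Input and Regions, clamped and cast to boolean, drives a Select that passes Input and Regions through unfiltered when polygons exist, otherwise outputting the created polygons.
; Clusters and Rawdata pins are unconnected.
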(graph InputsProcessing_SplineToPolygon
  (inputs (Input :type Any) (Clusters :type Point) (Areas :type Spline) (Regions :type Polygon2D) (Paths :type Spline) (Rawdata :type Any))
  ;[0 0 0 0.5] Inputs Processing and Transformation
  (bind FilterDataByType_0 (Filter%20Data%20By%20Type :TargetType (struct :Ids [(struct :Struct /Script/PCG.PCGDataTypeInfoPoint) (struct :Struct /Script/PCG.PCGDataTypeInfoSpline)])))
  (bind CreatePolygon2D_0 (Create%20Polygon%202D))
  ;[0.15 0.15 0.15 0.5] Is input already a polygon?
  (bind FilterDataByType_1 (Filter%20Data%20By%20Type :TargetType (struct :Ids [(struct :Struct /Script/PCG.PCGDataTypeInfoPolygon2D)])))
  (bind DataNum_0 (Data%20Count))
  (bind AttributeMathsOp_0 (Attribute%20Maths%20Op :Operation Clamp))
  (bind AttributeCast_0 (Attribute%20Cast :OutputType Boolean))
  (bind Select_0 (Select))
  (connect AttributeMathsOp_0 Out AttributeCast_0 In)
  (connect DataNum_0 Out AttributeMathsOp_0 In)
  (connect DefaultInputNode Areas CreatePolygon2D_0 In)
  (connect DefaultInputNode Paths CreatePolygon2D_0 In)
  (connect FilterDataByType_0 InsideFilter CreatePolygon2D_0 In)
  (connect FilterDataByType_1 InsideFilter DataNum_0 In)
  (connect DefaultInputNode Input FilterDataByType_0 In)
  (connect DefaultInputNode Input FilterDataByType_1 In)
  (connect DefaultInputNode Regions FilterDataByType_1 In)
  (connect AttributeCast_0 Out Select_0 bUseInputB)
  (connect CreatePolygon2D_0 Out Select_0 Input%20A)
  (connect DefaultInputNode Input Select_0 Input%20B)
  (connect DefaultInputNode Regions Select_0 Input%20B)
  (output Out Select_0))
